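# Sound cue for the beacon activation sound in the MASH limited-time mode.
# A single wave player feeds MASH_Beacon_Activate_01 straight to the output at reduced volume, with no randomization or mixing.
# It uses the important in-game mission sound class and a dedicated attenuation asset for long-range falloff.
# The cue also stores attenuation override values for distance low-pass filtering, reverb and stereo spread, though the override toggle is not shown enabled.
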SoundCue /Game/Athena/Sounds/LTMs/MASH/MASH_Beacon_Activate_Cue
root=WavePlayer_0
nodes [{id=WavePlayer_0, properties {soundWaveAssetPtr=&/MASH_Beacon_Activate_01}}]
properties {
 attenuationSettings=&/MASH_Beacon_Activate_Att
 maxDistance=20000
 soundClassObject=&Fort_Mission_Ingame_Important_PSM
 volumeMultiplier=0.6
 attenuationOverrides {
  bAttenuateWithLPF=true
  falloffDistance=12000
  lPFFrequencyAtMax=1000
  lPFRadiusMax=9500
  lPFRadiusMin=5000
  reverbDistanceMax=6000
  reverbDistanceMin=0
  reverbWetLevelMax=20
  stereoSpread=512
  attenuationShapeExtents {x=0}
 }
}
