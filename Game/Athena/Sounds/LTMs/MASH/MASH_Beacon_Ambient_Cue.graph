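# Sound cue for the looping ambient sound of the beacon in the MASH limited-time mode.
# A looping wave player feeds an Enveloper whose volume curve fades the loop in from silence.
# It plays in the Athena ambient props sound class with a boosted volume, spatialized through a dedicated attenuation asset.
# A concurrency override caps simultaneous instances.
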SoundCue /Game/Athena/Sounds/LTMs/MASH/MASH_Beacon_Ambient_Cue
root=Enveloper_0
nodes [
 {
  id=Enveloper_0
  children [WavePlayer_0]
  properties {pitchCurve {editorCurveData {keys [{time=0}]}}, volumeCurve {editorCurveData {keys [<at 0 0>,<at 1 1>]}}}
 }
 {id=WavePlayer_0, properties {bLooping=true, soundWaveAssetPtr=&/MASH_Beacon_Loop}}
]
properties {
 attenuationSettings=&/MASH_Beacon_Ambient_Loop_Att
 bOverrideConcurrency=true
 maxDistance=5000
 soundClassObject=&Fort_Ambient_Props_Athena
 volumeMultiplier=3
 attenuationOverrides {bEnableFocusInterpolation=true, bEnableListenerFocus=true, falloffDistance=2000, nonFocusVolumeAttenuation=0.5, stereoSpread=256}
 concurrencyOverrides {maxCount=3}
}
