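# Sound cue for the MASH Brute explosion, playing a layered one-shot explosion with randomized variation.
# A mixer plays three equal-volume layers at once, each picking randomly with equal weight from creature large-explode, BlackMonday bang, or energy-projectile explosion waves.
# The mix passes through an attenuation node that borrows the Deimos spawner large-ambient settings, and the cue is volume-boosted and routed to the Fort_AGV sound class.
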SoundCue /Game/Athena/Sounds/LTMs/MASH/MASH_Brute_Explode_Cue
root=Attenuation_0
nodes [
 {
  id=Attenuation_0
  children [Mixer_0]
  properties {attenuationSettings=/Game/Sounds/Fort_Monster_Deimos/Spawner/Fort_Deimos_SpawnerAmbientLarge_Att}
 }
 {id=Mixer_0, children [Random_2,Random_4,Random_3], properties {inputVolume [1,1,1]}}
 {
  id=Random_2
  children [WavePlayer_1, WavePlayer_9, WavePlayer_10, WavePlayer_11, WavePlayer_12, WavePlayer_13]
  properties {weights [1,1,1,1,1,1]}
 }
 {id=Random_3, children [WavePlayer_4,WavePlayer_5,WavePlayer_6], properties {weights [1,1,1]}}
 {id=Random_4, children [WavePlayer_7, WavePlayer_8, WavePlayer_14], properties {weights [1,1,1]}}
 {id=WavePlayer_1, properties {soundWaveAssetPtr=&/MASH_Creature_LargeExplode_01}}
 {id=WavePlayer_10, properties {soundWaveAssetPtr=&/MASH_Creature_LargeExplode_03}}
 {id=WavePlayer_11, properties {soundWaveAssetPtr=&/MASH_Creature_LargeExplode_04}}
 {id=WavePlayer_12, properties {soundWaveAssetPtr=&/MASH_Creature_LargeExplode_05}}
 {id=WavePlayer_13, properties {soundWaveAssetPtr=&/MASH_Creature_LargeExplode_06}}
 {id=WavePlayer_14, properties {soundWaveAssetPtr=/Game/Sounds/Fort_Weapons/BlackMonday/Bangs/BlackMonday_Bang_Explode_Alt_03}}
 {id=WavePlayer_4, properties {soundWaveAssetPtr=/Game/Sounds/Fort_Weapons/Impacts/EnergyProjectile/Alteration_EnergyProjectile_Explosion_01}}
 {id=WavePlayer_5, properties {soundWaveAssetPtr=/Game/Sounds/Fort_Weapons/Impacts/EnergyProjectile/Alteration_EnergyProjectile_Explosion_02}}
 {id=WavePlayer_6, properties {soundWaveAssetPtr=/Game/Sounds/Fort_Weapons/Impacts/EnergyProjectile/Alteration_EnergyProjectile_Explosion_03}}
 {id=WavePlayer_7, properties {soundWaveAssetPtr=/Game/Sounds/Fort_Weapons/BlackMonday/Bangs/BlackMonday_Bang_Explode_Alt_01}}
 {id=WavePlayer_8, properties {soundWaveAssetPtr=/Game/Sounds/Fort_Weapons/BlackMonday/Bangs/BlackMonday_Bang_Explode_Alt_02}}
 {id=WavePlayer_9, properties {soundWaveAssetPtr=&/MASH_Creature_LargeExplode_02}}
]
properties {maxDistance=4000, soundClassObject=/Game/Sounds/FortSoundClassesAndModes/Fort_AGV, volumeMultiplier=1.5}
