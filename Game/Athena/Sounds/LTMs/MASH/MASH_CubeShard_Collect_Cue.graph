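# Sound cue for collecting a cube shard in the MASH LTM, playing a random monster-parts pickup sound.
# A quality level node picks between a random pool of eight waves and a three-wave subset that reuses the same players, feeding a fixed-volume boost modulator.
# The cue adds its own volume boost, uses MASH pickup attenuation and caps concurrent instances, stopping the oldest.
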SoundCue /Game/Athena/Sounds/LTMs/MASH/MASH_CubeShard_Collect_Cue
root=Modulator_1
nodes [
 {id=Modulator_1, children [QualityLevel_0], properties {pitchMax=1, pitchMin=1, volumeMax=3, volumeMin=3}}
 {id=QualityLevel_0, children [Random_0,Random_0,Random_2,Random_2]}
 {
  id=Random_0
  children [WavePlayer_11, WavePlayer_12, WavePlayer_13, WavePlayer_14, WavePlayer_15, WavePlayer_16, WavePlayer_17, WavePlayer_18]
  properties {weights [1,1,1,1,1,1,1,1]}
 }
 {id=Random_2, children [WavePlayer_11, WavePlayer_14, WavePlayer_18], properties {bShouldExcludeFromBranchCulling=true, weights [1,1,1]}}
 {id=WavePlayer_11, properties {soundWaveAssetPtr=&/MASH_MonsterParts_Pickup_01}}
 {id=WavePlayer_12, properties {soundWaveAssetPtr=&/MASH_MonsterParts_Pickup_02}}
 {id=WavePlayer_13, properties {soundWaveAssetPtr=&/MASH_MonsterParts_Pickup_03}}
 {id=WavePlayer_14, properties {soundWaveAssetPtr=&/MASH_MonsterParts_Pickup_04}}
 {id=WavePlayer_15, properties {soundWaveAssetPtr=&/MASH_MonsterParts_Pickup_05}}
 {id=WavePlayer_16, properties {soundWaveAssetPtr=&/MASH_MonsterParts_Pickup_06}}
 {id=WavePlayer_17, properties {soundWaveAssetPtr=&/MASH_MonsterParts_Pickup_07}}
 {id=WavePlayer_18, properties {soundWaveAssetPtr=&/MASH_MonsterParts_Pickup_08}}
]
properties {
 attenuationSettings=/Game/Athena/Sounds/LTMs/MASH/Settings/Fort_MASH_Multiplier_Pickup_Att
 bOverrideConcurrency=true
 maxDistance=3500
 priority=0.5
 soundClassObject=&Fort_Loot_Reward_PSM
 volumeMultiplier=3
 concurrencyOverrides {maxCount=8, resolutionRule=StopOldest}
}
