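# Looping ambient sound cue for the MASH limited-time mode's difficulty increase, apparently played by an in-world beacon or prop.
# The idle loop wave runs through an Enveloper that ramps volume from silence to full, then a Modulator that applies a fixed volume boost without pitch variation.
# It uses the Athena ambient props sound class, the shared MASH beacon loop attenuation asset, and a concurrency override capping simultaneous instances.
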SoundCue /Game/Athena/Sounds/LTMs/MASH/MASH_DifficultyIncrease_Ambient_Cue
root=Modulator_0
nodes [
 {
  id=Enveloper_0
  children [WavePlayer_1]
  properties {pitchCurve {editorCurveData {keys [{time=0}]}}, volumeCurve {editorCurveData {keys [<at 0 0>,<at 1 1>]}}}
 }
 {id=Modulator_0, children [Enveloper_0], properties {pitchMax=1, pitchMin=1, volumeMax=2, volumeMin=2}}
 {id=WavePlayer_1, properties {bLooping=true, soundWaveAssetPtr=&/MASH_DifficultyIncrease_Idle_Loop}}
]
properties {
 attenuationSettings=&/MASH_Beacon_Ambient_Loop_Att
 bOverrideConcurrency=true
 maxDistance=5000
 soundClassObject=&Fort_Ambient_Props_Athena
 volumeMultiplier=3
 attenuationOverrides {bEnableFocusInterpolation=true, bEnableListenerFocus=true, falloffDistance=2000, nonFocusVolumeAttenuation=0.5, stereoSpread=256}
 concurrencyOverrides {maxCount=3}
}
