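# Sound cue for the MASH LTM explosive reload sound, classed as a medium explosive.
# A distance crossfade blends close explosion waves, which fade out with range, into distant variants that fade in farther away.
# Each layer gets slight pitch and volume modulation, then a quality switch picks random waves from five variants or a reduced set of three.
# Custom attenuation applies, and concurrency allows one instance, stopping the oldest.
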SoundCue /Game/Athena/Sounds/LTMs/MASH/MASH_ExplosiveReload_Cue
root=DistanceCrossFade_0
nodes [
 {
  id=DistanceCrossFade_0
  children [Modulator_8,Modulator_0]
  properties {crossFadeInput [
    {fadeInDistanceEnd=0, fadeInDistanceStart=0, fadeOutDistanceEnd=8000, fadeOutDistanceStart=3000, volume=1}
    {fadeInDistanceEnd=10000, fadeInDistanceStart=5000, fadeOutDistanceEnd=0, fadeOutDistanceStart=0, volume=1}
   ]}
 }
 {id=Modulator_0, children [QualityLevel_1], properties {pitchMax=1.1, volumeMax=1.2, volumeMin=1}}
 {id=Modulator_8, children [QualityLevel_0], properties {pitchMax=1.1, volumeMin=0.8}}
 {id=QualityLevel_0, children [Random_2,Random_2,Random_1,Random_1]}
 {id=QualityLevel_1, children [Random_0,Random_0,Random_3,Random_3]}
 {
  id=Random_0
  children [WavePlayer_15, WavePlayer_16, WavePlayer_17, WavePlayer_18, WavePlayer_19]
  properties {bShouldExcludeFromBranchCulling=true, weights [1,1,1,1,1]}
 }
 {id=Random_1, children [WavePlayer_6, WavePlayer_8, WavePlayer_11], properties {weights [1,1,1]}}
 {
  id=Random_2
  children [WavePlayer_6, WavePlayer_7, WavePlayer_8, WavePlayer_9, WavePlayer_11]
  properties {bShouldExcludeFromBranchCulling=true, weights [1,1,1,1,1]}
 }
 {id=Random_3, children [WavePlayer_15, WavePlayer_17, WavePlayer_19], properties {weights [1,1,1]}}
 {id=WavePlayer_11, properties {soundWaveAssetPtr=&/MASH_ExplodeReload_05}}
 {id=WavePlayer_15, properties {soundWaveAssetPtr=&/MASH_ExplodeReload_Distant_01}}
 {id=WavePlayer_16, properties {soundWaveAssetPtr=&/MASH_ExplodeReload_Distant_02}}
 {id=WavePlayer_17, properties {soundWaveAssetPtr=&/MASH_ExplodeReload_Distant_03}}
 {id=WavePlayer_18, properties {soundWaveAssetPtr=&/MASH_ExplodeReload_Distant_04}}
 {id=WavePlayer_19, properties {soundWaveAssetPtr=&/MASH_ExplodeReload_Distant_05}}
 {id=WavePlayer_6, properties {soundWaveAssetPtr=&/MASH_ExplodeReload_01}}
 {id=WavePlayer_7, properties {soundWaveAssetPtr=&/MASH_ExplodeReload_02}}
 {id=WavePlayer_8, properties {soundWaveAssetPtr=&/MASH_ExplodeReload_03}}
 {id=WavePlayer_9, properties {soundWaveAssetPtr=&/MASH_ExplodeReload_04}}
]
properties {
 attenuationSettings=/Game/Athena/Sounds/LTMs/MASH/Settings/ExplodeReload_Att
 bOverrideConcurrency=true
 maxDistance=10500
 priority=2
 soundClassObject=/Game/Sounds/FortSoundClassesAndModes/Fort_Explosives_Medium
 volumeMultiplier=0.5
 concurrencyOverrides {maxCount=1, resolutionRule=StopOldest, volumeScale=0.8, volumeScaleAttackTime=0.1}
 concurrencySet [null]
}
