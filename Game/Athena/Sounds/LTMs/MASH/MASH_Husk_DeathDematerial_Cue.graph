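# Sound cue for the dematerialize sound played when a husk dies in the MASH limited-time mode.
# A quality switch feeds two tiers from a random pick of ten wave variants and the other two from a three-variant subset that is excluded from branch culling.
# A modulator varies pitch on the result.
# It uses monster-death attenuation and overrides concurrency to a small cap that stops the oldest instance.
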SoundCue /Game/Athena/Sounds/LTMs/MASH/MASH_Husk_DeathDematerial_Cue
root=Modulator_0
nodes [
 {id=Modulator_0, children [QualityLevel_0], properties {pitchMax=1.3, pitchMin=0.7}}
 {id=QualityLevel_0, children [Random_0,Random_0,Random_1,Random_1]}
 {
  id=Random_0
  children [
   WavePlayer_6
   WavePlayer_7
   WavePlayer_8
   WavePlayer_9
   WavePlayer_10
   WavePlayer_11
   WavePlayer_12
   WavePlayer_13
   WavePlayer_14
   WavePlayer_15
  ]
  properties {weights [1,1,1,1,1,1,1,1,1,1]}
 }
 {id=Random_1, children [WavePlayer_6, WavePlayer_10, WavePlayer_15], properties {bShouldExcludeFromBranchCulling=true, weights [1,1,1]}}
 {id=WavePlayer_10, properties {soundWaveAssetPtr=&/MASH_Husk_Death_Dematerialize_05}}
 {id=WavePlayer_11, properties {soundWaveAssetPtr=&/MASH_Husk_Death_Dematerialize_06}}
 {id=WavePlayer_12, properties {soundWaveAssetPtr=&/MASH_Husk_Death_Dematerialize_07}}
 {id=WavePlayer_13, properties {soundWaveAssetPtr=&/MASH_Husk_Death_Dematerialize_08}}
 {id=WavePlayer_14, properties {soundWaveAssetPtr=&/MASH_Husk_Death_Dematerialize_09}}
 {id=WavePlayer_15, properties {soundWaveAssetPtr=&/MASH_Husk_Death_Dematerialize_10}}
 {id=WavePlayer_6, properties {soundWaveAssetPtr=&/MASH_Husk_Death_Dematerialize_01}}
 {id=WavePlayer_7, properties {soundWaveAssetPtr=&/MASH_Husk_Death_Dematerialize_02}}
 {id=WavePlayer_8, properties {soundWaveAssetPtr=&/MASH_Husk_Death_Dematerialize_03}}
 {id=WavePlayer_9, properties {soundWaveAssetPtr=&/MASH_Husk_Death_Dematerialize_04}}
]
properties {
 attenuationSettings=/Game/Athena/Sounds/LTMs/MASH/Settings/Fort_MASH_Monster_Death_Att
 bOverrideConcurrency=true
 maxDistance=2500
 soundClassObject=/Game/Sounds/FortSoundClassesAndModes/Fort_Enemy_Damaged_PSM
 volumeMultiplier=1.7
 attenuationOverrides {falloffDistance=5000, attenuationShapeExtents {x=1000}}
 concurrencyOverrides {maxCount=2, resolutionRule=StopOldest}
}
