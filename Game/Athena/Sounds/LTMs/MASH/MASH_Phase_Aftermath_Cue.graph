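# Sound cue that plays the music for the Aftermath phase of the MASH limited-time mode.
# A single wave player feeds the existing orchestral Combat_High_Threat_End track straight to the output.
# It uses an important in-game mission sound class, high priority and a huge max distance, and keeps playing virtually when silent.
# A concurrency override replaces shared settings and caps simultaneous instances at a small count.
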SoundCue /Game/Athena/Sounds/LTMs/MASH/MASH_Phase_Aftermath_Cue
root=WavePlayer_0
nodes [{id=WavePlayer_0, properties {soundWaveAssetPtr=/Game/Sounds/Fort_Music/Orchestral_Score/Orch_Combat/Combat_High_Threat_End}}]
properties {
 bOverrideConcurrency=true
 maxDistance=2097152
 priority=100
 soundClassObject=&Fort_Mission_Ingame_Important_PSM
 virtualizationMode=PlayWhenSilent
 concurrencyOverrides {maxCount=2}
}
