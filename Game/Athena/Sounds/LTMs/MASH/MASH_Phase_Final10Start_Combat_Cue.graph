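# Sound cue that plays high-threat orchestral combat music, named for the start of the Final 10 phase in the MASH limited-time mode.
# A fixed short delay feeds an envelope that fades the single music wave up from near silence over several seconds.
# High priority, play-when-silent virtualization and a very large max distance keep it active, while a concurrency override caps simultaneous copies.
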SoundCue /Game/Athena/Sounds/LTMs/MASH/MASH_Phase_Final10Start_Combat_Cue
root=Delay_0
nodes [
 {id=Delay_0, children [Enveloper_0], properties {delayMax=1.5, delayMin=1.5}}
 {
  id=Enveloper_0
  children [WavePlayer_2]
  properties {
   pitchCurve {editorCurveData {keys [{time=0}]}}
   volumeCurve {editorCurveData {keys [{arriveTangent=0.2193, interpMode=RCIM_Cubic, leaveTangent=0.2193, tangentMode=RCTM_User, time=0, value=0.01}, <at 8 1>]}}
  }
 }
 {id=WavePlayer_2, properties {soundWaveAssetPtr=/Game/Sounds/Fort_Music/Orchestral_Score/Orch_Combat/Combat_High_Threat_End_Long}}
]
properties {
 bOverrideConcurrency=true
 maxDistance=2097152
 priority=100
 soundClassObject=&Fort_Mission_Ingame_Important_PSM
 virtualizationMode=PlayWhenSilent
 volumeMultiplier=0.55
 concurrencyOverrides {maxCount=2}
}
