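# Sound cue for the final tick of a phase in the MASH limited-time mode, judging by its name and folder.
# A single wave player feeds the root with one UI tick wave, with no randomization or modulation.
# It plays at reduced volume under an important in-game mission sound class, with a very large max distance.
# A concurrency override caps how many instances play at once.
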SoundCue /Game/Athena/Sounds/LTMs/MASH/MASH_Phase_FinalTick_Cue
root=WavePlayer_0
nodes [{id=WavePlayer_0, properties {soundWaveAssetPtr=&/MASH_UI_Final_Tick_01}}]
properties {
 bOverrideConcurrency=true
 maxDistance=2097152
 soundClassObject=&Fort_Mission_Ingame_Important_PSM
 volumeMultiplier=0.4
 concurrencyOverrides {maxCount=2}
}
