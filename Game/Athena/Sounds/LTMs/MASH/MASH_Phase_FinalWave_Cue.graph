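# Sound cue for the final wave phase of the MASH limited-time mode, playing a very long high-threat combat-ending track from the orchestral score.
# A single wave player feeds the output directly at reduced volume, routed through an important in-game mission sound class.
# High priority, a huge max distance and play-when-silent virtualization keep it running, while a concurrency override caps it at two simultaneous instances.
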SoundCue /Game/Athena/Sounds/LTMs/MASH/MASH_Phase_FinalWave_Cue
root=WavePlayer_1
nodes [{id=WavePlayer_1, properties {soundWaveAssetPtr=/Game/Sounds/Fort_Music/Orchestral_Score/Orch_Combat/Combat_High_Threat_End_Very_Long}}]
properties {
 bOverrideConcurrency=true
 maxDistance=2097152
 priority=100
 soundClassObject=&Fort_Mission_Ingame_Important_PSM
 virtualizationMode=PlayWhenSilent
 volumeMultiplier=0.6
 concurrencyOverrides {maxCount=2}
}
