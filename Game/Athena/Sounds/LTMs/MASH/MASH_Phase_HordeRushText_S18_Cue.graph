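# Sound cue for the Horde Rush phase text in the MASH limited-time mode, apparently the Season 18 version.
# A Mixer blends a short high-threat orchestral combat intro with the phase text sound, which a fixed Delay holds back so it follows the music's start.
# It uses the important in-game mission sound class with high priority, a very large max distance and overridden concurrency that limits simultaneous instances.
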SoundCue /Game/Athena/Sounds/LTMs/MASH/MASH_Phase_HordeRushText_S18_Cue
root=Mixer_1
nodes [
 {id=Delay_0, children [WavePlayer_1], properties {delayMax=1.6, delayMin=1.6}}
 {id=Mixer_1, children [WavePlayer_0,Delay_0], properties {inputVolume [1.1,1.5]}}
 {id=WavePlayer_0, properties {soundWaveAssetPtr=/Game/Sounds/Fort_Music/Orchestral_Score/Orch_Combat/Combat_High_Threat_Intro_Short}}
 {id=WavePlayer_1, properties {soundWaveAssetPtr=&/MASH_Phase_HordeRushText_S18}}
]
properties {
 bOverrideConcurrency=true
 maxDistance=2097152
 priority=100
 soundClassObject=&Fort_Mission_Ingame_Important_PSM
 volumeMultiplier=0.5
 concurrencyOverrides {maxCount=2}
}
