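# Sound cue named for a phase intro in the MASH limited-time mode, with no sound nodes or root, so it plays no audio as it stands.
# It retains only playback settings: the important in-game mission sound class, reduced volume and a very large max distance.
# A concurrency override limits how many instances can play at once.
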SoundCue /Game/Athena/Sounds/LTMs/MASH/MASH_Phase_Intro_Cue
root=null
nodes []
properties {
 bOverrideConcurrency=true
 maxDistance=2097152
 soundClassObject=&Fort_Mission_Ingame_Important_PSM
 volumeMultiplier=0.6
 concurrencyOverrides {maxCount=2}
}
