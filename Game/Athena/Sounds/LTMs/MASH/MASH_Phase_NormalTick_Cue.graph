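# Sound cue for the normal phase tick in the MASH limited-time mode.
# A single wave player feeds a default UI tick sample straight to the output, raised in pitch and played at reduced volume.
# A concurrency override caps simultaneous instances at a low count, and the cue sets a very large max distance.
# It routes through the enemy-damaged sound class rather than one named for UI.
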SoundCue /Game/Athena/Sounds/LTMs/MASH/MASH_Phase_NormalTick_Cue
root=WavePlayer_0
nodes [{id=WavePlayer_0, properties {soundWaveAssetPtr=&/MASH_UI_Default_Tick_01}}]
properties {
 bOverrideConcurrency=true
 maxDistance=2097152
 pitchMultiplier=1.25
 soundClassObject=/Game/Sounds/FortSoundClassesAndModes/Fort_Enemy_Damaged_PSM
 volumeMultiplier=0.5
 concurrencyOverrides {maxCount=2}
}
