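# Sound cue for the preparation phase of the MASH limited-time mode, playing the very-low-threat orchestral combat intro music.
# A single wave player feeds an enveloper whose volume curve fades the music in over ten seconds.
# The cue plays at reduced volume in an important in-game mission sound class, with high priority, play-when-silent virtualization and its own concurrency cap.
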SoundCue /Game/Athena/Sounds/LTMs/MASH/MASH_Phase_Preparation_Cue
root=Enveloper_0
nodes [
 {
  id=Enveloper_0
  children [WavePlayer_0]
  properties {pitchCurve {editorCurveData {keys [{time=0}]}}, volumeCurve {editorCurveData {keys [<at 0 0>,<at 10 1>]}}}
 }
 {id=WavePlayer_0, properties {soundWaveAssetPtr=/Game/Sounds/Fort_Music/Orchestral_Score/Orch_Combat/Combat_Very_Low_Threat_Intro}}
]
properties {
 bOverrideConcurrency=true
 maxDistance=2097152
 priority=100
 soundClassObject=&Fort_Mission_Ingame_Important_PSM
 virtualizationMode=PlayWhenSilent
 volumeMultiplier=0.6
 concurrencyOverrides {maxCount=2}
}
