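# Sound cue for the phase round transition in the MASH limited-time mode.
# It waits a fixed one-second delay, then plays the reused front-end store-open UI sound at reduced volume.
# It borrows the enemy-damaged sound class, overrides concurrency to cap simultaneous instances at two, and sets a very large max distance.
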SoundCue /Game/Athena/Sounds/LTMs/MASH/MASH_Phase_RoundTransition_Cue
root=Delay_0
nodes [
 {id=Delay_0, children [WavePlayer_0], properties {delayMax=1, delayMin=1}}
 {id=WavePlayer_0, properties {soundWaveAssetPtr=/Game/Sounds/Fort_UI_Sounds/FrontEnd/fort_store_open}}
]
properties {
 bOverrideConcurrency=true
 maxDistance=2097152
 soundClassObject=/Game/Sounds/FortSoundClassesAndModes/Fort_Enemy_Damaged_PSM
 volumeMultiplier=0.4
 concurrencyOverrides {maxCount=2}
}
