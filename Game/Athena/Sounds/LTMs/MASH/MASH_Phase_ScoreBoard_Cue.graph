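# Sound cue for the MASH limited-time mode, named for a scoreboard phase.
# A single wave player reuses the frontend store-open UI sound at reduced volume.
# It overrides concurrency to cap simultaneous instances and is assigned the Fort_Enemy_Damaged_PSM sound class.
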SoundCue /Game/Athena/Sounds/LTMs/MASH/MASH_Phase_ScoreBoard_Cue
root=WavePlayer_0
nodes [{id=WavePlayer_0, properties {soundWaveAssetPtr=/Game/Sounds/Fort_UI_Sounds/FrontEnd/fort_store_open}}]
properties {
 bOverrideConcurrency=true
 maxDistance=2097152
 soundClassObject=/Game/Sounds/FortSoundClassesAndModes/Fort_Enemy_Damaged_PSM
 volumeMultiplier=0.3
 concurrencyOverrides {maxCount=2}
}
