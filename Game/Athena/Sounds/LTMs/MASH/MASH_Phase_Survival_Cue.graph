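# Sound cue that plays music for the survival phase of the MASH limited-time mode.
# A Random node picks with equal weight between two orchestral combat intro waves, high threat and very high threat.
# The cue runs at reduced volume with top priority, a very large max distance and play-when-silent virtualization.
# It routes to an important in-game mission sound class and overrides concurrency to cap simultaneous instances.
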SoundCue /Game/Athena/Sounds/LTMs/MASH/MASH_Phase_Survival_Cue
root=Random_0
nodes [
 {id=Random_0, children [WavePlayer_0,WavePlayer_1], properties {weights [1,1]}}
 {id=WavePlayer_0, properties {soundWaveAssetPtr=/Game/Sounds/Fort_Music/Orchestral_Score/Orch_Combat/Combat_High_Threat_Intro_Long}}
 {id=WavePlayer_1, properties {soundWaveAssetPtr=/Game/Sounds/Fort_Music/Orchestral_Score/Orch_Combat/Combat_Very_High_Threat_Intro_Long}}
]
properties {
 bOverrideConcurrency=true
 maxDistance=2097152
 priority=100
 soundClassObject=&Fort_Mission_Ingame_Important_PSM
 virtualizationMode=PlayWhenSilent
 volumeMultiplier=0.4
 concurrencyOverrides {maxCount=2}
}
