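# MASH_Phase_TickLoop_Cue is a Sound Cue in the MASH limited-time mode folder, named as a phase tick loop.
# It has no root or sound nodes, so it holds only settings and no audio to play.
# Those settings route it to the non-critical in-game UI sound class, lower its pitch, boost its volume and override concurrency to cap simultaneous instances.
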SoundCue /Game/Athena/Sounds/LTMs/MASH/MASH_Phase_TickLoop_Cue
root=null
nodes []
properties {
 bOverrideConcurrency=true
 maxDistance=2097152
 pitchMultiplier=0.7
 soundClassObject=&Fort_UI_InGame_NonCritical
 volumeMultiplier=1.5
 concurrencyOverrides {maxCount=2}
}
