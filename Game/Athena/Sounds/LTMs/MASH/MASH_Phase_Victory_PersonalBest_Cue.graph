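# Sound cue for the personal-best victory phase of the MASH limited-time mode, judging by its name.
# A mixer starts the one-shot personal-best wave at once and layers in a looping orchestral DBNO defeat music track after a fixed delay.
# It uses an important in-game mission sound class, high priority, play-when-silent virtualization and its own small concurrency limit.
# The looping branch means the cue does not end by itself.
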SoundCue /Game/Athena/Sounds/LTMs/MASH/MASH_Phase_Victory_PersonalBest_Cue
root=Mixer_0
nodes [
 {id=Delay_0, children [WavePlayer_0], properties {delayMax=4, delayMin=4}}
 {id=Mixer_0, children [Delay_0,WavePlayer_1], properties {inputVolume [1,1]}}
 {
  id=WavePlayer_0
  properties {bLooping=true, soundWaveAssetPtr=/Game/Sounds/Fort_Music/Orchestral_Score/Orch_Defeat/Defeat_Dying_DBNO_Main_Alt_80BPM_Loop}
 }
 {id=WavePlayer_1, properties {soundWaveAssetPtr=&/MASH_PersonalBest_01}}
]
properties {
 bOverrideConcurrency=true
 maxDistance=2097152
 priority=100
 soundClassObject=&Fort_Mission_Ingame_Important_PSM
 virtualizationMode=PlayWhenSilent
 volumeMultiplier=0.4
 concurrencyOverrides {maxCount=2}
}
