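# Sound cue for the Season 18 MASH LTM victory phase stinger.
# A mixer plays the MASH victory phase wave and an orchestral defeat stinger wave together at equal volume.
# The cue runs on an important in-game mission sound class at reduced volume, with a very large max distance and a concurrency override limiting simultaneous instances.
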SoundCue /Game/Athena/Sounds/LTMs/MASH/MASH_Phase_Victory_S18_Cue
root=Mixer_0
nodes [
 {id=Mixer_0, children [WavePlayer_2,WavePlayer_0], properties {inputVolume [1,1]}}
 {id=WavePlayer_0, properties {soundWaveAssetPtr=&/MASH_Phase_Victory_S18}}
 {id=WavePlayer_2, properties {soundWaveAssetPtr=/Game/Sounds/Fort_Music/Orchestral_Score/Orch_Defeat/Defeat_Stinger_01_Long_80BPM}}
]
properties {
 bOverrideConcurrency=true
 maxDistance=2097152
 soundClassObject=&Fort_Mission_Ingame_Important_PSM
 volumeMultiplier=0.4
 concurrencyOverrides {maxCount=2}
}
