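# Sound cue for the MASH limited-time mode, named for the victory-phase scoreboard, that plays looping orchestral music.
# It loops a wave from the defeat/DBNO score, despite the victory name, through an Enveloper that fades volume in over five seconds.
# The cue lowers overall volume, uses an important in-game mission sound class, and overrides concurrency to allow two instances.
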SoundCue /Game/Athena/Sounds/LTMs/MASH/MASH_Phase_Victory_ScoreBoard_Cue
root=Enveloper_0
nodes [
 {
  id=Enveloper_0
  children [WavePlayer_0]
  properties {pitchCurve {editorCurveData {keys [{time=0}]}}, volumeCurve {editorCurveData {keys [<at 0 0>,<at 5 1>]}}}
 }
 {
  id=WavePlayer_0
  properties {bLooping=true, soundWaveAssetPtr=/Game/Sounds/Fort_Music/Orchestral_Score/Orch_Defeat/Defeat_Dying_DBNO_Main_Alt_80BPM_Loop}
 }
]
properties {
 bOverrideConcurrency=true
 maxDistance=2097152
 soundClassObject=&Fort_Mission_Ingame_Important_PSM
 volumeMultiplier=0.4
 concurrencyOverrides {maxCount=2}
}
