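# Sound cue for the MASH LTM poison cloud, combining a one-shot burst with a sustained cloud loop.
# The burst layer mixes a random small creature explosion with a slightly delayed random poison burst, under a large Deimos spawner ambient attenuation.
# The looping cloud has its own small melee-weapon attenuation and a boosted mixer level.
# The whole cue plays at reduced volume in the large explosives sound class.
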SoundCue /Game/Athena/Sounds/LTMs/MASH/MASH_Poison_Cloud_Loop_Cue
root=Mixer_1
nodes [
 {
  id=Attenuation_0
  children [Mixer_0]
  properties {attenuationSettings=/Game/Sounds/Fort_Monster_Deimos/Spawner/Fort_Deimos_SpawnerAmbientLarge_Att}
 }
 {id=Attenuation_1, children [WavePlayer_0], properties {attenuationSettings=&Fort_Weapon_Melee_Sml_Att}}
 {id=Delay_0, children [Random_2], properties {delayMax=0.2, delayMin=0.2}}
 {id=Mixer_0, children [Delay_0,Random_1], properties {inputVolume [1,1]}}
 {id=Mixer_1, children [Attenuation_0, Attenuation_1], properties {inputVolume [1,2.5]}}
 {id=Random_1, children [WavePlayer_6,WavePlayer_7,WavePlayer_8], properties {weights [1,1,1]}}
 {id=Random_2, children [WavePlayer_1, WavePlayer_9, WavePlayer_10, WavePlayer_11], properties {weights [1,1,1,1]}}
 {id=WavePlayer_0, properties {bLooping=true, soundWaveAssetPtr=&/MASH_Poison_Cloud}}
 {id=WavePlayer_1, properties {soundWaveAssetPtr=&/MASH_Poison_Burst_04}}
 {id=WavePlayer_10, properties {soundWaveAssetPtr=&/MASH_Poison_Burst_02}}
 {id=WavePlayer_11, properties {soundWaveAssetPtr=&/MASH_Poison_Burst_03}}
 {id=WavePlayer_6, properties {soundWaveAssetPtr=&/MASH_Creature_SmallExplode_01}}
 {id=WavePlayer_7, properties {soundWaveAssetPtr=&/MASH_Creature_SmallExplode_02}}
 {id=WavePlayer_8, properties {soundWaveAssetPtr=&/MASH_Creature_SmallExplode_03}}
 {id=WavePlayer_9, properties {soundWaveAssetPtr=&/MASH_Poison_Burst_01}}
]
properties {maxDistance=4000, soundClassObject=/Game/Sounds/FortSoundClassesAndModes/Fort_Explosives_Large_PSM, volumeMultiplier=0.251}
