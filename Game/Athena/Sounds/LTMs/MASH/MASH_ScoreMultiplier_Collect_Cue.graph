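# Sound cue for the MASH limited-time mode, named for collecting a score multiplier pickup.
# A single wave player feeds the MASH_PickUp wave straight to the output at boosted volume.
# It uses the item-interact attenuation and the loot reward sound class.
# A concurrency override blocks new instances at the limit instead of stopping ones already playing.
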SoundCue /Game/Athena/Sounds/LTMs/MASH/MASH_ScoreMultiplier_Collect_Cue
root=WavePlayer_0
nodes [{id=WavePlayer_0, properties {soundWaveAssetPtr=&/MASH_PickUp}}]
properties {
 attenuationSettings=/Game/Sounds/Attenuation/Fort_Item_Interact_Att
 bOverrideConcurrency=true
 maxDistance=1850
 soundClassObject=&Fort_Loot_Reward_PSM
 volumeMultiplier=2
 concurrencyOverrides {resolutionRule=PreventNew}
}
